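; Art-tool Blueprint that keeps a byte Color and a float linearColor variable in sync, converting whichever one changes.
; The Construction Script compares each channel against cached copies, Color_0 and linearColor_0, with Color edits taking priority.
; A change converts to the other format, refreshes both caches and prints a debug string, "Hello" or "Bye" by direction.
; An update flag gates the check and is toggled around each write; event graph events are empty.
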(event EventBeginPlay)

(event Collision|EventActorBeginOverlap
  (params (OtherActor :type Actor%20Object%20Reference)))

(event EventTick
  (params (DeltaSeconds :type Float%20%28single-precision%29)))

(fn ConstructionScript
  (bind (_color_b _color_g _color_r _color_a) (Utilities|Struct|BreakColor (Variables|GetColor_B)))
  (bind (_color_0_b _color_0_g _color_0_r _color_0_a) (Utilities|Struct|BreakColor (Variables|GetColor_0_B)))
  (bind (_linearcolor_r _linearcolor_g _linearcolor_b _linearcolor_a) (Math|Color|BreakColor (Variables|GetlinearColor_R)))
  (bind (_linearcolor_0_r _linearcolor_0_g _linearcolor_0_b _linearcolor_0_a) (Math|Color|BreakColor (Variables|GetlinearColor_0_R)))
  (if (Variables|Default|Getupdate)
    (if (or (or (Math|Byte|NotEqual(Byte) _color_b _color_0_b) (Math|Byte|NotEqual(Byte) _color_g _color_0_g)) (or (Math|Byte|NotEqual(Byte) _color_r _color_0_r) (Math|Byte|NotEqual(Byte) _color_a _color_0_a)))
      (Variables|Default|Setupdate false)
      (bind _output_get (Variables|ignore|SetColor_0 (Variables|Default|GetColor)))
      (bind _output_get_2 (Variables|Default|SetlinearColor (Math|Conversions|ToLinearColor(Color) _output_get)))
      (Variables|ignore|SetlinearColor_0 _output_get_2)
      (Development|PrintString "Hello")
      (Variables|Default|Setupdate true)
      (elif (or (or (!= _linearcolor_r _linearcolor_0_r) (!= _linearcolor_g _linearcolor_0_g)) (or (!= _linearcolor_b _linearcolor_0_b) (!= _linearcolor_a _linearcolor_0_a)))
        (Variables|Default|Setupdate false)
        (bind _output_get_3 (Variables|ignore|SetlinearColor_0 (Variables|Default|GetlinearColor)))
        (bind _output_get_4 (Variables|Default|SetColor (Math|Conversions|ToColor(LinearColor) _output_get_3)))
        (Variables|ignore|SetColor_0 _output_get_4)
        (Development|PrintString "Bye")
        (Variables|Default|Setupdate true)))))
