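; Material library display actor that previews a three-to-five color gradient on a plane and a static mesh, labeled with its identifier text.
; The construction script sets the label and creates dynamic instances of MAT_Gmap_ML for both components when missing.
; It pushes Color0–Color4 and the five-color toggle to both, with useGmap off on the plane and on for the mesh.
; BeginPlay, overlap and Tick events are empty.
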(event EventBeginPlay)

(event Collision|EventActorBeginOverlap
  (params (OtherActor :type Actor%20Object%20Reference)))

(event EventTick
  (params (DeltaSeconds :type Float%20%28single-precision%29)))

(fn ConstructionScript
  (bind _dmi_plane (Variables|Default|GetDMI_Plane))
  (bind _dmi_geo (Variables|Default|GetDMI_Geo))
  (Rendering|Components|TextRender|SetText (Variables|Default|GetLabel) (Utilities|Text|ToText(String) (Variables|Default|Getidentifier)))
  (if (not (Utilities|IsValid (Variables|Default|GetDMI_Geo) :!class /Script/Engine.KismetSystemLibrary))
    (bind _returnvalue (Rendering|Material|CreateDynamicMaterialInstance (Variables|Default|GetStaticMesh) 0 "/Game/ArtTools/MaterialLibrary/Materials/MAT_Gmap_ML.MAT_Gmap_ML" "None" :!class /Script/Engine.PrimitiveComponent))
    (Variables|Default|SetDMI_Geo _returnvalue))
  (if (not (Utilities|IsValid (Variables|Default|GetDMI_Plane) :!class /Script/Engine.KismetSystemLibrary))
    (bind _returnvalue_2 (Rendering|Material|CreateDynamicMaterialInstance (Variables|Default|GetPlane) 0 "/Game/ArtTools/MaterialLibrary/Materials/MAT_Gmap_ML.MAT_Gmap_ML" "None" :!class /Script/Engine.PrimitiveComponent))
    (Variables|Default|SetDMI_Plane _returnvalue_2))
  (Utilities|FlowControl|Sequence
    (:then_0
      (Rendering|Material|SetVectorParameterValue _dmi_plane "Color0" (Variables|Default|GetColor0) :!class /Script/Engine.MaterialInstanceDynamic)
      (Rendering|Material|SetVectorParameterValue _dmi_plane "Color1" (Variables|Default|GetColor1) :!class /Script/Engine.MaterialInstanceDynamic)
      (Rendering|Material|SetVectorParameterValue _dmi_plane "Color2" (Variables|Default|GetColor2) :!class /Script/Engine.MaterialInstanceDynamic)
      (Rendering|Material|SetVectorParameterValue _dmi_plane "Color3" (Variables|Default|GetColor3) :!class /Script/Engine.MaterialInstanceDynamic)
      (Rendering|Material|SetVectorParameterValue _dmi_plane "Color4" (Variables|Default|GetColor4) :!class /Script/Engine.MaterialInstanceDynamic)
      (Rendering|Material|SetScalarParameterValue _dmi_plane "useFiveColors" (Math|Conversions|ToFloat(Boolean) (Variables|Default|GetFiveColors)) :!class /Script/Engine.MaterialInstanceDynamic)
      (Rendering|Material|SetScalarParameterValue _dmi_plane "useGmap" 0.0 :!class /Script/Engine.MaterialInstanceDynamic))
    (:then_1
      (Rendering|Material|SetVectorParameterValue _dmi_geo "Color0" (Variables|Default|GetColor0) :!class /Script/Engine.MaterialInstanceDynamic)
      (Rendering|Material|SetVectorParameterValue _dmi_geo "Color1" (Variables|Default|GetColor1) :!class /Script/Engine.MaterialInstanceDynamic)
      (Rendering|Material|SetVectorParameterValue _dmi_geo "Color2" (Variables|Default|GetColor2) :!class /Script/Engine.MaterialInstanceDynamic)
      (Rendering|Material|SetVectorParameterValue _dmi_geo "Color3" (Variables|Default|GetColor3) :!class /Script/Engine.MaterialInstanceDynamic)
      (Rendering|Material|SetVectorParameterValue _dmi_geo "Color4" (Variables|Default|GetColor4) :!class /Script/Engine.MaterialInstanceDynamic)
      (Rendering|Material|SetScalarParameterValue _dmi_geo "useFiveColors" (Math|Conversions|ToFloat(Boolean) (Variables|Default|GetFiveColors)) :!class /Script/Engine.MaterialInstanceDynamic)
      (Rendering|Material|SetScalarParameterValue _dmi_geo "useGmap" 1.0 :!class /Script/Engine.MaterialInstanceDynamic))))
