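; BP_MaterialSwatch is a material library swatch actor that previews character master material settings on a scalable mesh, with a text label.
; Construction sets the mesh, defaulting to a sphere, and calls RefreshMaterial, also exposed through a custom event.
; RefreshMaterial creates a dynamic instance if missing and pushes color, surface, mask and flatness variables into it.
; With a valid instance, an Identifier change updates the label and rebuilds a named instance from the child material.
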(event EventBeginPlay)

(event Collision|EventActorBeginOverlap
  (params (OtherActor :type Actor%20Object%20Reference)))

(event EventTick
  (params (DeltaSeconds :type Float%20%28single-precision%29)))

(event Custom|Refresh%20Material
  (CallFunction|RefreshMaterial))

(fn refreshMaterial
  (bind _identifier (Variables|MaterialParameters|GetIdentifier))
  (bind _internalidentifier (Variables|MaterialParameters|GetinternalIdentifier))
  (Variables|MaterialParameters|SetmaterialTypeInt (Math|Conversions|ToInteger(Byte) (Variables|MaterialParameters|GetMaterialType)))
  (if (Utilities|IsValid (Variables|MaterialParameters|GetMaterialInstanceDynamic) :!class /Script/Engine.KismetSystemLibrary)
    (if (Utilities|String|NotEqualExactly(String) _identifier _internalidentifier)
      (Development|PrintString (Utilities|String|Append "Comparing " _identifier " to " _internalidentifier))
      (Variables|MaterialParameters|SetinternalIdentifier _identifier)
      (CallFunction|SetText _identifier)
      (CallFunction|SetupDynamicMaterial (Utilities|String|StringToName (Utilities|String|Append "MID_" _identifier)) (Variables|MaterialParameters|GetBaseColor) (Variables|MaterialParameters|GetSpecular) (Variables|MaterialParameters|GetRoughness) (Variables|MaterialParameters|GetMetallic) (Variables|MaterialParameters|GetSkin) (Variables|MaterialParameters|GetSSSColor) (Variables|MaterialParameters|GetEmissive) (Variables|MaterialParameters|GetCloth) (Variables|MaterialParameters|GetClearCoat) (Variables|MaterialParameters|GetClearCoatRoughness) (Variables|Swatch|GetSwatchFlatness))
      (else
        _))
    (else
      (bind _returnvalue (Rendering|Material|CreateDynamicMaterialInstance (Variables|Default|GetSwatch) 0 "/Game/ArtTools/MaterialLibrary/Materials/M_FN_Character_MASTER_DSRM_parameters.M_FN_Character_MASTER_DSRM_parameters" (Utilities|String|StringToName (Utilities|String|Append "MID_" (Variables|MaterialParameters|GetIdentifier))) :!class /Script/Engine.PrimitiveComponent)) ; MaterialB - dynamic instance
      (bind _output_get (Variables|MaterialParameters|SetMaterialInstanceDynamic _returnvalue))
      (Rendering|Material|SetMaterial (Variables|Default|GetSwatch) 0 _output_get)))
  (CallFunction|SetMaterial (Variables|MaterialParameters|GetBaseColor) (Variables|MaterialParameters|GetSpecular) (Variables|MaterialParameters|GetRoughness) (Variables|MaterialParameters|GetMetallic) (Variables|MaterialParameters|GetSkin) (Variables|MaterialParameters|GetSSSColor) (Variables|MaterialParameters|GetEmissive) (Variables|MaterialParameters|GetCloth) (Variables|MaterialParameters|GetClearCoat) (Variables|MaterialParameters|GetClearCoatRoughness) (Variables|Swatch|GetSwatchFlatness))
  (Rendering|Material|SetMaterial (Variables|Default|GetSwatch) 0 (Variables|MaterialParameters|GetMaterialInstanceDynamic)))

(fn setMaterial
  (params (BaseColor :type Linear%20Color%20Structure) (S :type Float%20%28double-precision%29) (R :type Float%20%28double-precision%29) (M :type Float%20%28double-precision%29) (Skin :type Boolean) (SSSColor :type Linear%20Color%20Structure) (EmissiveColor :type Linear%20Color%20Structure) (ClothMask :type Boolean) (ClearCoatMask :type Boolean) (ClearCoatRoughness :type Float%20%28double-precision%29) (SwatchFlatness :type Float%20%28double-precision%29))
  (bind _materialinstancedynamic (Variables|MaterialParameters|GetMaterialInstanceDynamic))
  (Utilities|FlowControl|Sequence
    (:then_0
      (Rendering|Material|SetVectorParameterValue _materialinstancedynamic "BaseColor_MM" BaseColor :!class /Script/Engine.MaterialInstanceDynamic))
    (:then_1
      (Rendering|Material|SetScalarParameterValue _materialinstancedynamic "Specular_MM" S :!class /Script/Engine.MaterialInstanceDynamic))
    (:then_2
      (Rendering|Material|SetScalarParameterValue _materialinstancedynamic "Roughness_MM" R :!class /Script/Engine.MaterialInstanceDynamic))
    (:then_3
      (Rendering|Material|SetScalarParameterValue _materialinstancedynamic "Metallic_MM" M :!class /Script/Engine.MaterialInstanceDynamic))
    (:then_4
      (Rendering|Material|SetScalarParameterValue _materialinstancedynamic "SkinMask_MM" (Math|Conversions|ToFloat(Boolean) Skin) :!class /Script/Engine.MaterialInstanceDynamic)
      (Rendering|Material|SetVectorParameterValue _materialinstancedynamic "Skin Boost Color and Exponent" SSSColor :!class /Script/Engine.MaterialInstanceDynamic))
    (:then_5
      (Rendering|Material|SetVectorParameterValue _materialinstancedynamic "EmissiveColor_MM" EmissiveColor :!class /Script/Engine.MaterialInstanceDynamic))
    (:then_6
      (Rendering|Material|SetScalarParameterValue _materialinstancedynamic "ClothMask_MM" (Math|Conversions|ToFloat(Boolean) ClothMask) :!class /Script/Engine.MaterialInstanceDynamic))
    (:then_7
      (Rendering|Material|SetScalarParameterValue _materialinstancedynamic "ClearCoat_MM" (Math|Conversions|ToFloat(Boolean) ClearCoatMask) :!class /Script/Engine.MaterialInstanceDynamic)
      (Rendering|Material|SetScalarParameterValue _materialinstancedynamic "ClearCoatRoughness_MM" ClearCoatRoughness :!class /Script/Engine.MaterialInstanceDynamic))
    (:then_8
      (Rendering|Material|SetScalarParameterValue _materialinstancedynamic "Flatness" SwatchFlatness :!class /Script/Engine.MaterialInstanceDynamic))))

(fn setMaterialParamsBP
  (params (BaseColor :type Linear%20Color%20Structure) (S :type Float%20%28double-precision%29) (R :type Float%20%28double-precision%29) (M :type Float%20%28double-precision%29) (Skin :type Boolean) (SSSColor :type Linear%20Color%20Structure) (EmissiveColor :type Linear%20Color%20Structure) (ClothMask :type Boolean) (ClearCoatMask :type Boolean) (ClearCoatRoughness :type Float%20%28double-precision%29) (SwatchFlatness :type Float%20%28double-precision%29))
  (Utilities|FlowControl|Sequence
    (:then_0
      (Variables|MaterialParameters|SetBaseColor BaseColor))
    (:then_1
      (Variables|MaterialParameters|SetSpecular S))
    (:then_2
      (Variables|MaterialParameters|SetRoughness R))
    (:then_3
      (Variables|MaterialParameters|SetMetallic M))
    (:then_4
      (Variables|MaterialParameters|SetSkin Skin)
      (Variables|MaterialParameters|SetSSSColor SSSColor))
    (:then_5
      (Variables|MaterialParameters|SetEmissive EmissiveColor))
    (:then_6
      (Variables|MaterialParameters|SetCloth ClothMask))
    (:then_7
      (Variables|MaterialParameters|SetClearCoat ClearCoatMask))
    (:then_8
      (Variables|MaterialParameters|SetClearCoatRoughness ClearCoatRoughness))
    (:then_9
      (Variables|Swatch|SetSwatchFlatness SwatchFlatness))))

(fn setText
  (params (text :type String))
  (Variables|MaterialParameters|SetIdentifier text)
  (Rendering|Components|TextRender|SetText (Variables|Default|GetTextRender) (Utilities|Text|ToText(String) text))
  (bind _output_get (Variables|MaterialParameters|SetinternalIdentifier text))
  (Development|PrintString (Utilities|String|Append "setting internalIdentifier " _output_get)))

(fn setupDynamicMaterial
  (params (Name :type Name) (BaseColor :type Linear%20Color%20Structure) (S :type Float%20%28double-precision%29) (R :type Float%20%28double-precision%29) (M :type Float%20%28double-precision%29) (Skin :type Boolean) (SSSColor :type Linear%20Color%20Structure) (EmissiveColor :type Linear%20Color%20Structure) (ClothMask :type Boolean) (ClearCoatMask :type Boolean) (ClearCoatRoughness :type Float%20%28double-precision%29) (SwatchFlatness :type Float%20%28double-precision%29))
  (Utilities|FlowControl|Sequence
    (:then_0
      (bind _returnvalue (Rendering|Material|CreateDynamicMaterialInstance (Variables|Default|GetSwatch) 0 "/Game/ArtTools/MaterialLibrary/Materials/M_FN_Character_MASTER_DSRM_parameters_child.M_FN_Character_MASTER_DSRM_parameters_child" Name :!class /Script/Engine.PrimitiveComponent)) ; MaterialB - dynamic instance
      (bind _output_get (Variables|MaterialParameters|SetMaterialInstanceDynamic _returnvalue))
      (Rendering|Material|SetMaterial (Variables|Default|GetSwatch) 0 _output_get))
    (:then_1
      (CallFunction|SetMaterial BaseColor S R M Skin SSSColor EmissiveColor ClothMask ClearCoatMask ClearCoatRoughness SwatchFlatness))
    (:then_2
      (CallFunction|SetMaterialParamsBP BaseColor S R M Skin SSSColor EmissiveColor ClothMask ClearCoatMask ClearCoatRoughness SwatchFlatness))))

(fn ConstructionScript
  (bind _swatch (Variables|Default|GetSwatch))
  (bind _swatchmesh (Variables|Swatch|GetSwatchMesh))
  (if (Utilities|IsValid _swatchmesh :!class /Script/Engine.KismetSystemLibrary)
    (Components|StaticMesh|SetStaticMesh _swatch _swatchmesh)
    (else
      (Components|StaticMesh|SetStaticMesh _swatch "/Engine/BasicShapes/Sphere.Sphere")))
  (CallFunction|RefreshMaterial)
  (Transformation|SetRelativeScale3D (Variables|Default|GetSwatch) (Math|Vector|MakeVector (Variables|Swatch|GetSwatchScale) (Variables|Swatch|GetSwatchScale) (Variables|Swatch|GetSwatchScale))))
