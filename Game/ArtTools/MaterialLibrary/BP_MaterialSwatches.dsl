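; BP_MaterialSwatches is an editor tool actor that manages a material library of BP_MaterialSwatch and BP_Gradient child actors.
; Custom events spawn swatches from a DSRM data table (gamma-converting 0-255 colors) or from parameters, add gradients, or adopt selected swatches.
; LayoutMaterials arranges and relabels them in a grid or labeled per-material-type rows.
; Python scripts handle file import, export and label sorting, while the construction script applies a shared swatch mesh and scale.
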(event EventBeginPlay)

(event Collision|EventActorBeginOverlap
  (params (OtherActor :type Actor%20Object%20Reference)))

(event EventTick
  (params (DeltaSeconds :type Float%20%28single-precision%29)))

(event Custom|Create%20Swatches%20From%20Table
  (bind _materialswatches (Variables|Default|GetMaterialSwatches))
  ;[1 1 1] Cleanup Swatches plugged into this Blueprint
  (for _array_element _materialswatches
    (Actor|DestroyActor _array_element))
  (Utilities|Array|Clear _materialswatches)
  ;[1 1 1] Make Swatches
  (bind _outrownames (DataTable|GetDataTableRowNames (Variables|Default|GetDataTable_read)))
  (for _array_element_2 _outrownames
    (bind _returnvalue (CallFunction|CreateMaterialSwatch (Utilities|String|ToString(Name) _array_element_2)))
    (Transformation|AttachActorToActor _returnvalue self "None" "KeepWorld" "KeepRelative" "KeepRelative" true)
    (bind _returnvalue_2 (Utilities|GetDataTableRow (Variables|Default|GetDataTable_read) _array_element_2))
    (Class|BPMaterialSwatch|SetupDynamicMaterial :self _returnvalue :Name (Utilities|String|StringToName (Utilities|String|Append "MID_" (Utilities|String|ToString(Name) _array_element_2))) :BaseColor (Math|Color|MakeColor (Math|Float|Power (/ (Math|Conversions|ToFloat(Integer) (.R_11_5716E1FC4D0DB88CD865EBAE38441C32 (Utilities|Struct|BreakDSRMStructure _returnvalue_2))) 255.0) 2.2) (Math|Float|Power (/ (Math|Conversions|ToFloat(Integer) (.G_13_2331315E4F9826E1706A3E9CEA914E79 (Utilities|Struct|BreakDSRMStructure _returnvalue_2))) 255.0) 2.2) (Math|Float|Power (/ (Math|Conversions|ToFloat(Integer) (.B_15_CBF5810E4254C55A121BA5895A88729A (Utilities|Struct|BreakDSRMStructure _returnvalue_2))) 255.0) 2.2)) :S (.Spec_21_1764241948C0BA2A17115D83A5B35F6F (Utilities|Struct|BreakDSRMStructure _returnvalue_2)) :R (.Rough_28_A12952E24EAA68D271B1C7A3150E5DB3 (Utilities|Struct|BreakDSRMStructure _returnvalue_2)) :M (.Metallic_26_7B5E966644D2E3E7E0A65294E0D4D64E (Utilities|Struct|BreakDSRMStructure _returnvalue_2)) :Skin false :ClothMask false :ClearCoatMask false :ClearCoatRoughness 0.0 :SwatchFlatness 0.0)))

(event Custom|Export%20to%20File
  (bind _csv_exportfilepath (Variables|Default|Getcsv_exportFilePath))
  (bind _filepath (Utilities|Struct|BreakFilePath _csv_exportfilepath))
  (bind _returnvalue (Utilities|Paths|GetPath _filepath))
  (if (Utilities|String|NotEqualExactly(String) (Utilities|Paths|GetCleanFilename _filepath))
    (if (not (Utilities|Paths|FileExists _filepath))
      (Development|PrintString 0 (Utilities|String|Append "Checking for directory " _returnvalue) true true "(R=1.000000,G=0.009844,B=0.000000,A=1.000000)")
      (if (not (Utilities|Paths|DirectoryExists _returnvalue))
        (Development|PrintString 0 (Utilities|String|Append "Can not export. " _filepath " does not exist") true true "(R=1.000000,G=0.009844,B=0.000000,A=1.000000)")))
    (Development|ExecuteConsoleCommand 0 (Utilities|String|Append "py TechArt/MaterialLibrary/MaterialSwatches_export.py " (Utilities|GetDisplayName self) " " _filepath))
    (else
      (Development|PrintString 0 "No filename given" true true "(R=1.000000,G=0.009844,B=0.000000,A=1.000000)"))))

(event Custom|ImportFromFile
  ;[1 1 1] Import from file was used to bring in the odler .csv library
  (CallFunction|SetLeveltoBPLevel)
  (Python|Execution|ExecutePythonCommand (Utilities|String|Append "TechArt/MaterialLibrary/MaterialSwatches_import.py " (Utilities|GetDisplayName self))))

(event Custom|AddNewGradient
  (CallFunction|SetLeveltoBPLevel)
  (bind _returnvalue (EditorScripting|LevelUtility|SpawnActorfromClass "/Game/ArtTools/MaterialLibrary/BP_Gradient.BP_Gradient_C" :!class /Script/EditorScriptingUtilities.EditorLevelLibrary))
  (Transformation|AttachActorToActor _returnvalue self "None" "KeepWorld" "KeepRelative" "KeepRelative" true)
  (Utilities|Array|Add (Variables|Default|GetGradients) _returnvalue)
  (CallFunction|LayoutMaterials))

(event Custom|Add%20New%20Swatch
  (CallFunction|SetLeveltoBPLevel)
  (bind _returnvalue (CallFunction|CreateMaterialSwatch "New Material"))
  (Transformation|AttachActorToActor _returnvalue self "None" "KeepWorld" "KeepRelative" "KeepRelative" true)
  (Class|BPMaterialSwatch|SetupDynamicMaterial :self _returnvalue :Name "New Material" :BaseColor "(R=0.200000,G=0.200000,B=0.200000,A=1.000000)" :S 0.5 :R 0.5 :M 0.0 :Skin false :ClothMask false :ClearCoatMask false :ClearCoatRoughness 0.0 :SwatchFlatness 0.0))

(event Custom|AddNewSwatchWithParameters
  (params (BaseColor :type Linear%20Color%20Structure) (S :type Float%20%28double-precision%29) (R :type Float%20%28double-precision%29) (M :type Float%20%28double-precision%29))
  (Development|PrintString 0 (Utilities|String|ToString(LinearColor) BaseColor))
  (bind _returnvalue (CallFunction|CreateMaterialSwatch "New Material"))
  (Transformation|AttachActorToActor _returnvalue self "None" "KeepWorld" "KeepRelative" "KeepRelative" true)
  (Class|BPMaterialSwatch|SetupDynamicMaterial :self _returnvalue :Name "New Material" :BaseColor BaseColor :S S :R R :M M :Skin false :ClothMask false :ClearCoatMask false :ClearCoatRoughness 0.0 :SwatchFlatness 0.0))

(event Custom|LayoutMaterials
  (bind _materialtypecounter (Variables|Default|GetmaterialTypeCounter))
  (bind _numacross (Variables|Default|GetnumAcross))
  (bind _spacing (Variables|Default|Getspacing))
  (for _array_element (Variables|Default|GetMaterialSwatches)
    (Transformation|AttachActorToActor _array_element self "None" "KeepRelative" "KeepRelative" "KeepRelative" true))
  (CallFunction|CreateMaterialTypeText)
  (CallFunction|SortMaterialsbyLabels)
  ;[1 1 1] relayout the materials based on type & labels
  ;;[1 1 1] Clear materialTypeCounter .. (tracks how many of each type we have)
  (Utilities|Array|Clear _materialtypecounter)
  (for _index (range (+ (Utilities|Enum|GetnumberofentriesinEnum_MaterialTypes) 1))
    (Utilities|Array|SetArrayElem _materialtypecounter _index 1 true))
  (Utilities|Array|Clear (Variables|Default|GetMaterialSwatches_tmp))
  (for _array_element_2 (Variables|Default|GetMaterialSwatches)
    (if (not (Class|ItemUIFunctionLibrary|IsNullItem _array_element_2))
      (Utilities|Array|Add (Variables|Default|GetMaterialSwatches_tmp) _array_element_2)))
  (for (_array_element_3 _array_index) (Variables|Default|GetMaterialSwatches_tmp)
    (if (Variables|Default|GetsortByMaterialType)
      (bind _returnvalue (CallFunction|CalculateTransformMaterialType _array_index _numacross _spacing (Variables|/Game/ArtTools/MaterialLibrary/BP_MaterialSwatch.BP_MaterialSwatch_C|GetMaterialType _array_element_3)))
      (Variables|Default|SetmaterialTransform _returnvalue)
      (else
        (bind _returnvalue_2 (CallFunction|CalculateTransform _array_index _numacross _spacing _spacing))
        (Variables|Default|SetmaterialTransform _returnvalue_2)))
    (Transformation|SetActorRelativeTransform _array_element_3 (Variables|Default|GetmaterialTransform) false false)
    (EditorScripting|ActorEditing|SetActorLabel _array_element_3 (Utilities|String|Append "BP_MaterialSwatch_" (Variables|/Game/ArtTools/MaterialLibrary/BP_MaterialSwatch.BP_MaterialSwatch_C|GetIdentifier _array_element_3)) true))
  ;[1 1 1] Gradients
  (for (_array_element_4 _array_index_2) (Variables|Default|GetGradients)
    (bind _returnvalue_3 (CallFunction|CalculateTransform _array_index_2 (Variables|Default|GetnumAcross) (Variables|Default|Getspacing) (* (Variables|Default|Getspacing) 2.0)))
    (Transformation|SetActorRelativeTransform _array_element_4 (Math|Transform|MakeTransform (+ (* (.location _returnvalue_3) "1.000000,1.000000,-1.000000") "50.000000,0.0,60.000000") (.rotation _returnvalue_3) (.scale _returnvalue_3)) false false)
    (EditorScripting|ActorEditing|SetActorLabel _array_element_4 (Utilities|String|Append "BP_Gradient_" (Variables|/Game/ArtTools/MaterialLibrary/BP_Gradient.BP_Gradient_C|Getidentifier _array_element_4)) true)))

(event Custom|HookupSelectedMaterials
  (Utilities|Array|Clear (Variables|Default|GetMaterialSwatches))
  (bind _returnvalue (EditorScripting|LevelUtility|GetSelectedLevelActors))
  (bind _returnvalue_2 (EditorScripting|Utilities|Filter|FilterbyClass _returnvalue "/Game/ArtTools/MaterialLibrary/BP_MaterialSwatch.BP_MaterialSwatch_C"))
  (Variables|Default|SetMaterialSwatches _returnvalue_2))

(fn CalculateTransform
  (params (index :type Integer) (numAcross :type Integer) (spacingX :type Float%20%28double-precision%29) (spacingZ :type Float%20%28double-precision%29))
  (return (Math|Transform|MakeTransform (Math|Vector|MakeVector (+ (+ (* (Math|Integer|%25%28Integer%29 index numAcross) spacingX) spacingX) 3e+01) 0.0 (* (/ index numAcross) (* spacingZ -1.0))) "0, 0, 0" "1.000000,1.000000,1.000000")))

(fn CalculateTransform_materialType
  (params (index :type Integer) (numAcross :type Integer) (spacing :type Float%20%28double-precision%29) (materialType :type Enum_MaterialTypes%20Enum :default "NewEnumerator5"))
  (bind _returnvalue (Math|Conversions|ToInteger(Byte) materialType))
  (bind _result (Math|Integer|IncrementInt (Utilities|Array|Get(aref) (Variables|Default|GetmaterialTypeCounter) _returnvalue))
    (:%20%20
      (return (Math|Transform|MakeTransform (Math|Vector|MakeVector (+ (* _result spacing) 4e+01) 0.0 (* (* _returnvalue spacing) -1.0)) "0, 0, 0" "1.000000,1.000000,1.000000")))))

(fn CreateMaterialSwatch
  (params (Name :type String))
  (bind _returnvalue (Game|SpawnActorfromClass 0 "/Game/ArtTools/MaterialLibrary/BP_MaterialSwatch.BP_MaterialSwatch_C" (Transformation|GetActorTransform) "Undefined" "OverrideRootScale" 0 0))
  (bind _returnvalue_2 (Utilities|Array|Add (Variables|Default|GetMaterialSwatches) _returnvalue))
  (bind _returnvalue_3 (CallFunction|CalculateTransform _returnvalue_2 (Variables|Default|GetnumAcross) (Variables|Default|Getspacing)))
  (Transformation|AddActorLocalTransform _returnvalue _returnvalue_3 false true)
  (Class|BPMaterialSwatch|SetText _returnvalue Name)
  (return _returnvalue))

(fn CreateMaterialTypeText
  (bind _materiallabels (Variables|Default|GetmaterialLabels))
  (CallFunction|SetLeveltoBPLevel)
  (for _array_element _materiallabels
    (Actor|DestroyActor _array_element))
  (Utilities|Array|Clear _materiallabels)
  (if (Variables|Default|GetsortByMaterialType)
    (bind _enumvalue (Utilities|Enum|ForEachEnum_MaterialTypes false)
      (:LoopBody
        (bind _returnvalue (Game|SpawnActorfromClass 0 "/Script/Engine.TextRenderActor" (Math|Transform|MakeTransform (Math|Vector|MakeVector 1e+02 0.0 (- (* (Math|Conversions|ToInteger(Byte) _enumvalue) (* (Variables|Default|Getspacing) -1.0)) (* (Variables|Default|Getspacing) 0.2))) "0,90.000000, 0" "1.000000,1.000000,1.000000") "Undefined" "MultiplyWithRoot" 0 0))
        (Utilities|Array|Add (Variables|Default|GetmaterialLabels) _returnvalue)
        (Rendering|Material|SetMaterial (Variables|/Script/Engine.TextRenderActor|GetTextRender _returnvalue) 0 "/Game/ArtTools/MaterialLibrary/Materials/DefaultTextMaterial_emissive.DefaultTextMaterial_emissive")
        (Transformation|AttachActorToActor _returnvalue self "None" "KeepRelative" "KeepRelative" "KeepRelative" true)
        (Rendering|Components|TextRender|SetText (Variables|/Script/Engine.TextRenderActor|GetTextRender _returnvalue) (Utilities|Text|ToText(String) (Utilities|String|EnumtoString _enumvalue)))
        (Rendering|Components|TextRender|SetHorizontalAlignment (Variables|/Script/Engine.TextRenderActor|GetTextRender _returnvalue) "EHTA_Right")
        (Rendering|Components|TextRender|SetWorldSize (Variables|/Script/Engine.TextRenderActor|GetTextRender _returnvalue) 12.0)
        (EditorScripting|ActorEditing|SetActorLabel _returnvalue (Utilities|String|Append "TextRender_" (Utilities|String|EnumtoString _enumvalue)) true)))))

(fn getLevel
  (params (object :type Object%20Reference))
  (bind (_lefts _rights) (Utilities|String|Split (Utilities|GetObjectPathString object) "/" "IgnoreCase" "FromEnd"))
  (return (Utilities|String|Split _rights ".")))

(fn setBaseMaterialType
  (for _array_element (Variables|Default|GetMaterialSwatches)
    (if (> (Variables|/Game/ArtTools/MaterialLibrary/BP_MaterialSwatch.BP_MaterialSwatch_C|GetMetallic _array_element) 0.1)
      (Variables|/Game/ArtTools/MaterialLibrary/BP_MaterialSwatch.BP_MaterialSwatch_C|SetBaseMaterialType "NewEnumerator0" _array_element)
      (else
        (Variables|/Game/ArtTools/MaterialLibrary/BP_MaterialSwatch.BP_MaterialSwatch_C|SetBaseMaterialType "NewEnumerator1" _array_element)))))

(fn setLevelToBPLevel
  (bind (_returnvalue _level) (Level|getLevel 0 self))
  (EditorScripting|LevelUtility|SetCurrentLevelByName (Utilities|String|StringToName _level)))

(fn SortMaterialsByLabels
  (Development|ExecuteConsoleCommand 0 (Utilities|String|Append "py TechArt/MaterialLibrary/MaterialSwatches_sort.py " (Utilities|GetDisplayName self))))

(fn ConstructionScript
  (bind _swatchmeshsize (Variables|Default|GetSwatchMeshSize))
  (if (Utilities|IsValid (Variables|Default|GetSwatchMesh) :!class /Script/Engine.KismetSystemLibrary)
    (for _array_element (Variables|Default|GetMaterialSwatches)
      ;[1 1 1] Comment
      (Variables|/Game/ArtTools/MaterialLibrary/BP_MaterialSwatch.BP_MaterialSwatch_C|SetSwatchScale _swatchmeshsize _array_element)
      (Components|StaticMesh|SetStaticMesh (Variables|/Game/ArtTools/MaterialLibrary/BP_MaterialSwatch.BP_MaterialSwatch_C|GetSwatch _array_element) (Variables|Default|GetSwatchMesh)))
    (else
      (for _array_element_2 (Variables|Default|GetMaterialSwatches)
        (Variables|/Game/ArtTools/MaterialLibrary/BP_MaterialSwatch.BP_MaterialSwatch_C|SetSwatchScale _swatchmeshsize _array_element_2)
        (Components|StaticMesh|SetStaticMesh (Variables|/Game/ArtTools/MaterialLibrary/BP_MaterialSwatch.BP_MaterialSwatch_C|GetSwatch _array_element_2) "/Engine/BasicShapes/Sphere.Sphere")))))
