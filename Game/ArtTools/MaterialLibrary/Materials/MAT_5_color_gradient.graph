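# An unlit material in the art tools material library, named for a five-color gradient.
# It holds three texture samples: a three-color gradient, a five-color brick gradient and a five-color gradient.
# The graph shows no links between these samples or into any material output.
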Material /Game/ArtTools/MaterialLibrary/Materials/MAT_5_color_gradient
graph {
 properties {pixelDepthOffsetMode=PDOM_Legacy, shadingModel=MSM_Unlit, usedShadingModels=Unlit}
 nodes [
  TextureSample() {texture=&/3_Color_Gradient}
  TextureSample() {texture=&/5_Color_Brick_Gradient}
  TextureSample() {texture=&/5_Color_Gradient}
 ]
}
