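# MF_3ColorGrad_ML is a material function in the material library that maps a scalar input to a three-colour gradient.
# It blends three colour vector parameters, defaulting to red, green and blue, through the Lerp_3Color function using the input as alpha.
# The blended result is raised to a constant gamma-style power before output, darkening the midtones as in an sRGB-to-linear conversion.
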MaterialFunction /Game/ArtTools/MaterialLibrary/Materials/MF_3ColorGrad_ML
graph {nodes [
  n1 = VectorParameter() {parameterName=Color0, defaultValue=<rgba 1 0 0 1>}
  n2 = VectorParameter() {parameterName=Color1, defaultValue=<rgba 0 1 0 1>}
  n3 = VectorParameter() {parameterName=Color2, defaultValue=<rgba 0 0 1 1>}
  n4 = FunctionInput() {inputType=FunctionInput_Scalar}
  FunctionOutput(None=Power(Base=MaterialFunctionCall(A=@n1.RGB, B=@n2.RGB, C=@n3.RGB, Alpha=@n4) {materialFunction=&Lerp_3Color}.Result) {constExponent=2.2})
 ]}
