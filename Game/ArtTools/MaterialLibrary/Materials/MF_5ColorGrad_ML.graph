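# Material function that maps a scalar input onto a five-color gradient built from the Color0 to Color4 inputs.
# It chains two Lerp_3Color calls, one per half of the range, with Color2 as the shared midpoint and the scalar rescaled for each half.
# An If at 0.5 picks the matching half, and a final 2.2 power applies a gamma-style curve to the resulting color.
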MaterialFunction /Game/ArtTools/MaterialLibrary/Materials/MF_5ColorGrad_ML
graph {nodes [
  n1 = Power(Base=FunctionInput() {inputType=FunctionInput_Scalar}) {constExponent=1}
  n2 = FunctionInput() {inputName=Color2}
  n3 = Multiply(A=@n1) {constB=2}
  n4 = MaterialFunctionCall(A=@n2, B=FunctionInput() {inputName=Color3}, C=FunctionInput() {inputName=Color4}, Alpha=Add(A=@n3) {constB=-1}) {materialFunction=&Lerp_3Color}
  n5 = MaterialFunctionCall(A=FunctionInput() {inputName=Color0}, B=FunctionInput() {inputName=Color1}, C=@n2, Alpha=@n3) {materialFunction=&Lerp_3Color}
  FunctionOutput(None=Power(Base=If(A=@n1, "A > B"=@n4.Result, "A == B"=@n5.Result, "A < B"=@n5.Result) {constB=0.5}) {constExponent=2.2})
 ]}
