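# MAT_Gmap_ML is an unlit Material Library material in ArtTools that builds a color gradient map.
# A useGmap scalar switches the gradient input between the red channel of the T_MasterKey_Swatch_GMap grayscale texture and a LinearGradient U ramp.
# A useFiveColors scalar then picks MF_5ColorGrad_ML, fed by five color parameters, or MF_3ColorGrad_ML, which receives only that input.
# Both switches are If nodes comparing against a constant threshold, not static switches.
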Material /Game/ArtTools/MaterialLibrary/Materials/MAT_Gmap_ML
graph {
 properties {pixelDepthOffsetMode=PDOM_Legacy, shadingModel=MSM_Unlit, usedShadingModels=Unlit}
 nodes [
  Constant3Vector()
  n1 = If(
   A=ScalarParameter() {parameterName=useGmap}
   "A > B"=TextureSample() {samplerType=SAMPLERTYPE_Grayscale, texture=&/T_MasterKey_Swatch_GMap}.R
   "A < B"=MaterialFunctionCall() {materialFunction=&LinearGradient}.UGradient
  ) {constB=0.5}
  If(
   A=ScalarParameter() {parameterName=useFiveColors}
   "A > B"=MaterialFunctionCall!(
    Color0=VectorParameter() {parameterName=Color0, defaultValue=<rgba 1 0 0 1>}.RGB
    Color1=VectorParameter() {parameterName=Color1, defaultValue=<rgba 1 1 0 1>}.RGB
    Color2=VectorParameter() {parameterName=Color2, defaultValue=<rgba 0 1 0 1>}.RGB
    Color3=VectorParameter() {parameterName=Color3, defaultValue=<rgba 0 1 1 1>}.RGB
    Color4=VectorParameter() {parameterName=Color4, defaultValue=<rgba 0 0 1 1>}.RGB
    In=@n1
   ) {materialFunction=&/MF_5ColorGrad_ML}.Result
   "A < B"=MaterialFunctionCall!(In=@n1) {materialFunction=&/MF_3ColorGrad_ML}.Result
  ) {constB=0.5}
 ]
}
