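# Character material function that builds material attributes from flat DSRM parameters for base color, metallic, specular, roughness and clear coat, plus a flattened normal map.
# An If on ClearCoat_MM selects the ClearCoat shading model above 0.5, otherwise the default one.
# Extra outputs expose normal, roughness, diffuse, transmission color and the SpecularMasks texture.
# Diffuse and mask textures, SSS brightness and mobile roughness remapping are present but reach no output.
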MaterialFunction /Game/ArtTools/MaterialLibrary/Materials/BaseFN_CharacterFunction_DSRM_parameters
graph {nodes [
  n1 = TextureSampleParameter2D() {
   desc="R = SpecMask G = Metallic B = Roughness"
   group=Specular
   parameterName=SpecularMasks
   samplerType=SAMPLERTYPE_Masks
   texture=&T_Fortnite_Default_S
  }
  n2 = LinearInterpolate(
   A=ScalarParameter() {group=Roughness, parameterName=RoughnessMin}
   B=ScalarParameter() {defaultValue=1, group=Roughness, parameterName=RoughnessMax}
   Alpha=@n1.B
  )
  n3 = StaticSwitchParameter(
   True=LinearInterpolate(
    A=ScalarParameter() {group=Roughness, parameterName=MobileRoughnessMin}
    B=ScalarParameter() {defaultValue=1, group=Roughness, parameterName=MobileRoughnessMax}
    Alpha=@n1.B
   )
   False=@n2
  ) {parameterName=OverrideRoughnessOnMobile}
  FeatureLevelSwitch(Default=@n2, ES2=@n3, ES3_1=@n3, SM4_REMOVED=@n2)
  n4 = VectorParameter() {desc=single color baseColor, group=DSRM Parameter, parameterName=BaseColor_MM, defaultValue=<rgba 0.3 0.3 0.3 1>}
  n5 = ScalarParameter() {defaultValue=0.5, group=DSRM Parameter, parameterName=Roughness_MM}
  n6 = Constant()
  n7 = MaterialFunctionCall(
   Normal=TextureSampleParameter2D() {group=Normal, parameterName=Normals, samplerType=SAMPLERTYPE_Normal, texture=/Engine/EngineMaterials/DefaultNormal}.RGB
   Flatness=ScalarParameter() {defaultValue=1, parameterName=Flatness, sliderMax=2}
  ) {materialFunction=&FlattenNormal}
  n8 = ScalarParameter() {defaultValue=1, desc=ClearCoat value, group=DSRM Parameter, parameterName=ClearCoat_MM, sliderMax=1}
  n9 = TextureSampleParameter2D() {group=Hair, parameterName=Tangent Map, texture=/Game/Packages/Fortress_SharedMaterials/Textures/FLOW_MAP_TEXTURE}
  n10 = Constant()
  n11 = ShadingModel() {shadingModel=MSM_ClearCoat}
  n12 = ShadingModel()
  n13 = VectorParameter() {desc=single color baseColor, group=DSRM Parameter, parameterName=Transmission_MM, defaultValue=<rgba 1 1 0 1>}
  FunctionOutput(None=MakeMaterialAttributes!(
    BaseColor=@n4.RGB
    Metallic=ScalarParameter() {defaultValue=0.5, group=DSRM Parameter, parameterName=Metallic_MM}
    Specular=ScalarParameter() {defaultValue=0.5, group=DSRM Parameter, parameterName=Specular_MM}
    Roughness=@n5
    Anisotropy
    EmissiveColor=@n6
    Opacity=@n13.RGB
    OpacityMask
    Normal=@n7.Result
    Tangent
    WorldPositionOffset
    SubsurfaceColor=@n6
    ClearCoat=@n8
    ClearCoatRoughness=ScalarParameter() {desc=ClearCoat Roughness, group=DSRM Parameter, parameterName=ClearCoatRoughness_MM}
    AmbientOcclusion
    Refraction=LinearInterpolate(A=@n9.RGB, B=@n10) {constAlpha=1}
    CustomizedUVs_0
    CustomizedUVs_1
    CustomizedUVs_2
    CustomizedUVs_3
    CustomizedUVs_4
    CustomizedUVs_5
    CustomizedUVs_6
    CustomizedUVs_7
    PixelDepthOffset
    ShadingModel=If(A=@n8, "A > B"=@n11, "A == B"=@n12, "A < B"=@n12) {constB=0.5}
    Displacement
   ))
  FunctionOutput(None=@n7.Result) {outputName=Normal Map, sortPriority=2}
  FunctionOutput(None=@n5) {outputName=Roughness, sortPriority=8}
  FunctionOutput(None=@n1.RGB) {outputName=SpecularMask, sortPriority=3}
  FunctionOutput(None=@n13.RGB) {outputName=M, sortPriority=4}
  FunctionOutput(None=@n4.RGB) {outputName=DIffuse, sortPriority=1}
  n14 = TextureSampleParameter2D() {
   desc="Masks R=AO G= Fresnel B = Tmission/Skin"
   group=Masks
   parameterName=M
   samplerType=SAMPLERTYPE_Masks
   texture=/Engine/EngineMaterials/DefaultDiffuse_TC_Masks
  }
  n15 = Multiply(A=@n14.B, B=ScalarParameter() {defaultValue=1, group=SkinColor, parameterName=SSS Brightness})
  Multiply(A=@n15)
  Reroute(None=@n15)
  Reroute(None=@n1.R)
  Reroute(None=@n1.G)
  StaticSwitch(True=@n9.RGB, False=@n10)
  TextureSampleParameter2D() {group=Diffuse, parameterName=Diffuse, texture=/Engine/EngineMaterials/DefaultDiffuse}
 ]}
