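# Material function that procedurally applies a wet look to character materials by modifying base color and roughness of incoming Material Attributes.
# It estimates porosity from the surface itself, with rougher, non-metallic areas scoring higher, so no mask texture is needed.
# That mask darkens base color through Wetness Multiplier and pulls roughness toward zero, both scaled by the Wetness parameter.
# Metals and already-smooth areas stay largely unchanged.
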MaterialFunction /Game/Packages/Fortress_SharedMaterials/Base_Material_Functions/BaseCharacterShaderFunctions/MF_CharacterWetnessProcedural
graph {nodes [
  n1 = GetMaterialAttributes(None=FunctionInput() {inputName=Material In, inputType=FunctionInput_MaterialAttributes}) {attributeGetTypes [#1, #2, #3]}
  n2 = OneMinus(None=@n1.Metallic)
  n3 = Constant() {r=0.4}
  n4 = Saturate(None=Divide(A=Subtract(A=@n1.Roughness, B=@n3) {constB=0.5}, B=Constant() {r=0.4}) {constB=0.4})
  n5 = ComponentMask(None=Saturate(None=Multiply(A=@n2, B=@n4))) {r=true}
  n6 = Multiply(
   A=@n1.BaseColor
   B=LinearInterpolate(B=ScalarParameter() {defaultValue=0.2, parameterName=Wetness Multiplier}, Alpha=@n5) {constA=1, constB=0.2}
  )
  n7 = ScalarParameter() {group=Wetness, parameterName=Wetness}
  n8 = LinearInterpolate(B=LinearInterpolate(Alpha=@n5) {constA=1, constB=0}, Alpha=@n7) {constA=1}
  FunctionOutput(None=SetMaterialAttributes!(
    MaterialAttributes=@n1.MaterialAttributes
    BaseColor=LinearInterpolate(A=@n1.BaseColor, B=@n6, Alpha=@n7)
    Roughness=LinearInterpolate(B=@n1.Roughness, Alpha=@n8)
   ) {attributeSetTypes [#1, #3]})
 ]}
