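# Character master material for skinned, morph-target and cloth meshes, built on material attributes with per-pixel shading model selection.
# A parameter-driven DSRM base supplies packed AO/GMap/skin masks to chained GMap color, balance, emissive, skin and skin FX functions.
# EmissiveColor_MM overrides emissive before skin, and rim is re-added afterward, feeding de-rez and opacity on the default static switch branch.
# A procedural wetness call branches off with no consumer.
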Material /Game/ArtTools/MaterialLibrary/Materials/M_FN_Character_MASTER_DSRM_parameters
graph {
 properties {
  bAutomaticallySetUsageInEditor=false
  bUseMaterialAttributes=true
  bUsedWithClothing=true
  bUsedWithMeshDeformer=true
  bUsedWithMorphTargets=true
  bUsedWithSkeletalMesh=true
  pixelDepthOffsetMode=PDOM_Legacy
  refractionMethod=RM_IndexOfRefraction
  shadingModel=MSM_FromMaterialExpression
  subsurfaceProfile=&SS_HeroSkin_02
  usedShadingModels=Default Lit | Subsurface | Clear Coat | Subsurface Profile | Cloth
 }
 nodes [
  n1 = MaterialFunctionCall() {materialFunction=&/BaseFN_CharacterFunction_DSRM_parameters}
  n2 = MaterialFunctionCall(Float3=@n1.M) {desc="R = AO\r\nG = GMap\r\nB = Skin", materialFunction=&BreakOutFloat3Components}
  n3 = MaterialFunctionCall!(Material=@n1.Result, Diffuse=@n1.Result, GMap=@n2.G) {materialFunction=/Game/Packages/Fortress_SharedMaterials/Base_Material_Functions/BaseCharacterShaderFunctions/BaseFN_CharacterGMapColor}
  n4 = MaterialFunctionCall!(Material=@n3.Result) {materialFunction=/Game/Packages/Fortress_SharedMaterials/Base_Material_Functions/CharacterBrightnessandScale_Balance}
  n5 = PixelNormalWS()
  n6 = CameraVectorWS()
  n7 = MaterialFunctionCall!(
   Material=@n4.Result
   Diffuse=@n3.DIffuse
   BalanceDistance=@n4.Result
   Normals=@n1.Normal Map
   AO=@n2.R
   Fresnel=Saturate(None=DotProduct(A=@n5, B=@n6))
  ) {materialFunction=/Game/Packages/Fortress_SharedMaterials/Base_Material_Functions/BaseCharacterShaderFunctions/BaseFN_CharacterEmissive}
  n8 = VectorParameter() {parameterName=EmissiveColor_MM}
  n9 = MaterialFunctionCall!(
   Gmap=@n2.G
   MaterialAttributes=MaterialFunctionCall!(
    Material=SetMaterialAttributes(MaterialAttributes=@n7.Result, Emissive Color=@n8.RGB) {attributeSetTypes [#1]}
    AO=@n2.R
    Skin Mask=ScalarParameter() {parameterName=SkinMask_MM}
   ) {materialFunction=/Game/Packages/Fortress_SharedMaterials/Base_Material_Functions/BaseCharacterShaderFunctions/BaseFN_CharacterSkin}.Result
   AO=@n2.R
  ) {materialFunction=&/CharacterSkinFX_DRSM}
  MaterialFunctionCall!(Material In=@n9.Result) {materialFunction=/Game/Packages/Fortress_SharedMaterials/Base_Material_Functions/BaseCharacterShaderFunctions/MF_CharacterWetnessProcedural}
  n10 = GetMaterialAttributes(None=@n9.Result) {attributeGetTypes [#1]}
  n11 = SetMaterialAttributes(MaterialAttributes=@n10.MaterialAttributes, Emissive Color=Add(A=@n10.Emissive Color, B=@n7.Rim)) {attributeSetTypes [#1]}
  StaticSwitch(
   True=MaterialFunctionCall!(Material=MaterialFunctionCall!(Fresnel, Material=@n11) {materialFunction=/Game/Packages/Fortress_SharedMaterials/Base_Material_Functions/CharacterDeRez_Transition}.Result) {materialFunction=/Game/Packages/Fortress_SharedMaterials/Base_Material_Functions/BaseCharacterShaderFunctions/BaseFN_CharacterOpacity}.Result
   False=SetMaterialAttributes!(MaterialAttributes, Shading Model=ShadingModel() {shadingModel=MSM_SubsurfaceProfile}) {attributeSetTypes [#2]}
  ) {defaultValue=true}
 ]
}
